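# Parent material for the default HighTower glider, adding an animated screen effect to the shared glider base shader and weapon skin FX function.
# Custom-primitive-data timestamps drive a U/V sphere-mask sweep that reveals the screen with a bright edge.
# Flickering base color plus masked, panning two-layer electric noise forms the screen content, modulated by distance-faded scanlines and Fresnel brightness.
# The mask texture's red channel confines it before tech-art overrides.
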Material /BRCosmetics/Gadgets/Assets/VinderTech_GliderChute/Glider_HighTower_Default/Material/M_HighTower_Parent
graph {
 properties {
  bUseMaterialAttributes=true
  bUsedWithMeshDeformer=true
  bUsedWithSkeletalMesh=true
  pixelDepthOffsetMode=PDOM_Legacy
  usedShadingModels=Default Lit
 }
 nodes [
  Constant()
  Constant()
  n1 = GetMaterialAttributes(None=MaterialFunctionCall!(
    Pre-Skinned World Normal
    Gmap
    Customization Mask
    IsCharacter
    IsStatsB
    Material Attributes=MaterialFunctionCall!(UV 1 coordinates) {materialFunction=/Game/Gadgets/Assets/VinderTech_GliderChute/Materials/MF_GliderBaseShader}.Result Material Attributes
   ) {materialFunction=/Game/Weapons/WeaponSkins/MaterialFunction/Weapon_SkinFXFunction}.Result) {attributeGetTypes [#1, #2]}
  n2 = TextureCoordinate()
  n3 = ScalarParameter() {bUseCustomPrimitiveData=true, parameterName=TimestampA1}
  n4 = ScalarParameter() {bUseCustomPrimitiveData=true, defaultValue=0.3, parameterName=TransitionLength_In1, primitiveDataIndex=2}
  n5 = ScalarParameter() {defaultValue=0.45, parameterName=SweepHardness}
  n6 = MaterialFunctionCall(
   TimestampA=ScalarParameter() {bUseCustomPrimitiveData=true, parameterName=TimestampA2, primitiveDataIndex=1}
   TransitionLength_In=ScalarParameter() {defaultValue=0.9, parameterName=TransitionLength_In2}
  ) {materialFunction=&MF_Timestamp_InOut_01}
  n7 = ScalarParameter() {defaultValue=0.005, parameterName=VRadiusBias}
  n8 = Min(
   A=SphereMask(
    A=ComponentMask(None=@n2) {r=true}
    B=ScalarParameter() {defaultValue=0.4, parameterName=SweepU}
    Radius=MaterialFunctionCall(TimestampA=@n3, TransitionLength_In=@n4) {materialFunction=&MF_Timestamp_InOut_01}.Linear
    Hardness=@n5
   )
   B=SphereMask(
    A=ComponentMask(None=@n2) {g=true}
    B=ScalarParameter() {defaultValue=0.8, parameterName=SweepV}
    Radius=Add(A=@n6.Linear, B=@n7)
    Hardness=LinearInterpolate(A=@n5, Alpha=@n6.Linear)
   )
  )
  n9 = ScalarParameter() {defaultValue=0.5, parameterName=SweepEdgeCenter}
  n10 = ScalarParameter() {defaultValue=0.5, parameterName=SweepEdgeRadius}
  n11 = Constant()
  n12 = Time()
  n13 = Multiply(
   A=Add(A=Sine(None=@n12) {period=0.3981}, B=Sine(None=@n12) {period=0.5889})
   B=ScalarParameter() {defaultValue=0.165, parameterName=FlickerScale}
  )
  n14 = TextureCoordinate()
  n15 = VectorParameter() {parameterName=ElectricNoiseTilePanA, defaultValue=<rgba 0.275 0.275 0.0212 0.0712>}
  n16 = AppendVector(A=@n15.R, B=@n15.R)
  n17 = TextureSample(UVs=Panner(Coordinate=Multiply(A=@n14, B=@n16), Speed=AppendVector(A=@n15.B, B=@n15.B)) {bFractionalPart=true}) {samplerType=SAMPLERTYPE_LinearColor, texture=/Game/Global/Textures/Default/Noise/T_Curie_NoisePack}
  n18 = TextureCoordinate()
  n19 = VectorParameter() {parameterName=ElectricNoiseTilePanB, defaultValue=<rgba 0.594 0.594 -0.0385 0.1717>}
  n20 = AppendVector(A=@n19.R, B=@n19.R)
  n21 = TextureSample(UVs=Panner(Coordinate=Multiply(A=@n18, B=@n20), Speed=AppendVector(A=@n19.B, B=@n19.B)) {bFractionalPart=true}) {samplerType=SAMPLERTYPE_LinearColor, texture=&T_Curie_NoisePack}
  n22 = Add(
   A=Power(Base=Add(A=@n17.G, B=@n21.B), Exp=ScalarParameter() {defaultValue=2.75, parameterName=ElectricExponent})
   B=ScalarParameter() {defaultValue=0.15, parameterName=ElectricBias}
  )
  n23 = VectorParameter() {parameterName=ElectricColor, defaultValue=<rgba 0.0208 0.0858 0.8 0>}
  n24 = Add(
   A=Add(A=Multiply(A=@n1.Base Color, B=@n13), B=@n1.Base Color)
   B=Multiply(
    A=TextureSample() {texture=/BRCosmetics/Gadgets/Assets/VinderTech_GliderChute/Glider_HighTower_Default/Textures/Skins/T_HighTower_Default_M}.G
    B=Multiply(A=@n22, B=@n23.RGB)
   )
  )
  n25 = Multiply(
   A=Power(Base=SphereMask(A=@n8, B=@n9, Radius=@n10, Hardness=@n11), Exp=ScalarParameter() {defaultValue=8, parameterName=SweepEdgeExponent})
   B=@n24
  )
  n26 = ScalarParameter() {defaultValue=100, parameterName=SweepEdgeBrightness}
  n27 = TextureSample() {texture=&T_HighTower_Default_M}
  n28 = LinearInterpolate(
   A=@n1.Base Color
   B=Multiply(A=Multiply(A=@n8, B=@n24), B=ScalarParameter() {defaultValue=0.5, parameterName=DiffuseScreenScale})
   Alpha=@n27.R
  ) {constA=1}
  n29 = ScalarParameter() {defaultValue=0.4, parameterName=ScanlineLowValue}
  n30 = ScalarParameter() {defaultValue=1.9, parameterName=ScanlineHighValue}
  n31 = ComponentMask(None=Panner(Coordinate=TextureCoordinate(), Speed=ScalarParameter() {defaultValue=-0.022, parameterName=ScanlinePanSpeed}) {bFractionalPart=true}) {g=true}
  n32 = ScalarParameter() {defaultValue=160, parameterName=ScanlineRepeat}
  n33 = SphereMask(
   A=Frac(None=Multiply(A=@n31, B=@n32))
   B=Constant() {r=0.5}
   Radius=ScalarParameter() {defaultValue=0.35, parameterName=ScanlineRadius}
   Hardness=ScalarParameter() {defaultValue=0.2, parameterName=ScanlineHarndess}
  )
  n34 = MaterialFunctionCall() {materialFunction=&ObjectPivotPoint}
  n35 = ScalarParameter() {defaultValue=900, parameterName=Fade_Start_Distance}
  n36 = ScalarParameter() {defaultValue=500, parameterName=Fade_End_Distance}
  n37 = ScalarParameter() {defaultValue=8, parameterName=FresnelExponent}
  n38 = Multiply(
   A=LinearInterpolate(
    A=LinearInterpolate(A=@n29, B=@n30, Alpha=@n33)
    Alpha=MaterialFunctionCall(Position=@n34.Object Pivot Location, Fade_Start_Distance=@n35, Fade_End_Distance=@n36) {materialFunction=&MF_Distance_SyncedFade_FX}.Result
   )
   B=LinearInterpolate(
    A=ScalarParameter() {defaultValue=3, parameterName=EmissiveScaleLow}
    B=ScalarParameter() {defaultValue=8, parameterName=EmissiveScaleHigh}
    Alpha=MaterialFunctionCall(Exponent=@n37) {materialFunction=&MF_Fresnel_FX}.Fresnel
   )
  )
  n39 = Multiply(A=Add(A=Multiply(A=@n25, B=@n26), B=Multiply(A=@n28, B=@n38)), B=@n27.R)
  MaterialFunctionCall(In=SetMaterialAttributes!(MaterialAttributes=@n1.MaterialAttributes, Emissive Color=Add(A=@n1.Emissive Color, B=@n39), Base Color=@n28) {attributeSetTypes [#1, #2]}) {materialFunction=&MF_TechArtOverrides_accessories}
  MaterialFunctionCall!(Time, Divisor=ScalarParameter() {defaultValue=3, parameterName=DivisorB}, Mod) {materialFunction=/Game/Effects/Fort_Effects/Materials/_Functions/MF_ModuloTime}
  MaterialFunctionCall(Divisor=ScalarParameter() {defaultValue=0.5, parameterName=DivisorA}, Mod=ScalarParameter() {defaultValue=6, parameterName=ModA}) {materialFunction=&MF_ModuloTime}
 ]
}
